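# Animation Blueprint for the Sushi Chef pickaxe that adds procedural secondary motion to its dyn_loop bone.
# The input pose passes through a single Anim Dynamics node in component space, then returns to local space for output.
# A stiff, heavily damped angular spring with tight angular limits and zero gravity scale constrains the simulated motion.
# The Event Graph holds only unused default nodes, and no variables are defined.
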AnimBlueprint /BRCosmetics/Weapons/FORT_Melee/Pickaxe_Sushi_Chef/Meshes/Pickaxe_Sushi_Chef_AnimBP
target_skeleton=&/Pickaxe_Sushi_Chef_Skeleton
graphs [
 AnimGraph {nodes [
   AnimGraphNode_SubInput_0 = Input Pose()
   n1 = Anim Dynamics - Bone: dyn_loop(ComponentPose=Local To Component(LocalPose=@AnimGraphNode_SubInput_0), Alpha=1, ExternalForce="0,0,0") {
    angularDampingOverride=0.98
    angularSpringConstant=1000
    bAngularSpring=true
    bGravityOverrideInSimSpace=true
    bOverrideAngularDamping=true
    gravityScale=0
    numSolverIterationsPostUpdate=4
    numSolverIterationsPreUpdate=16
    alphaBoolBlend {blendOption=Linear}
    boundBone {boneName=dyn_loop}
    componentLinearAccScale=<1 1 1>
    physicsBodyDefinitions [{
      sphereCollisionRadius=0
      boundBone {boneName=dyn_loop}
      boxExtents=<5 5 5>
      constraintSetup {angularLimitsMax=<10 10 10>, angularLimitsMin=<-10 -10 -10>, angularTarget=<1 0 0>}
      localJointOffset=<0 0 6>
     }]
   }
   Output Pose(Result=Component To Local(ComponentPose=@n1)) {layerGroup=None}
  ]}
 EventGraph {nodes [
   TryGetPawnOwner()
   Event BlueprintUpdateAnimation()
  ]}
]
variables []
